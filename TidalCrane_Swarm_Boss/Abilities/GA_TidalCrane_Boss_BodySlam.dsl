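; Gameplay ability for the TidalCrane boss's body slam: a telegraphed montage whose notify triggers target selection, damage and knockback of pawns and vehicles.
; When rocks are enabled, a shuffled EQS query plus upward traces picks overhead spawn points and places warning actors.
; After the slam, a delayed ground-pound cue plays and projectiles drop from those points.
; Activation needs a valid target, range-limited only when rocks are disabled.
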(event Ability|EventActivateAbility
  (bind _returnvalue (Ability|GetAvatarActorfromActorInfo))
  (bind (_montageinfo_montagetoplay _montageinfo_animplayrate _montageinfo_animrootmotiontranslationscale _montageinfo_montagesectiontoplay _montageinfo_overridesection) (Utilities|Struct|BreakFortGameplayAbilityMontageInfo (Variables|GetMontageInfo_MontageToPlay)))
  (Ability|CommitAbility)
  (Variables|InternalData|SetBossActor _returnvalue)
  (Default|SetAllowLookatTarget(Message) _returnvalue false)
  (Ability|AddGameplayCueToOwner (Variables|Default|GetGC_BodySlamTelegraph))
  (if (ScalableFloat|GetValueasBool (Variables|Default|Projectiles|GetCanSpawnRockProjectiles))
    (CallFunction|CalculateRockFallingPositions)
    (else
      _))
  (bind (_asynctaskproxy _notifytag) (Ability|Tasks|PlayMontageAndWaitNotify "None" _montageinfo_montagetoplay _montageinfo_animplayrate _montageinfo_overridesection true _montageinfo_animrootmotiontranslationscale)
    (:OnCompleted
      _)
    (:OnInterrupted
      _)
    (:OnCancelled
      _)
    (:OnNotifyBegin
      (if (== _notifytag (Variables|Default|GetTag_BodySlam))
        (bind (_asynctaskproxy_2 _targetdata _applicationtag) (Ability|Tasks|DoTargetSelection "None" "(TagName=\"Targeting.ApplicationTag.0\")")
          (:then
            (Variables|Default|SetPoundVFXSpawnLocation (+ (Transformation|GetActorLocation (Ability|GetAvatarActorfromActorInfo)) (* (Transformation|GetActorForwardVector (Ability|GetAvatarActorfromActorInfo)) 8.5e+02)))
            (Utilities|FlowControl|Delay 0.1)
            (Ability|ExecuteGameplayCueWithParamsOnOwner (Variables|Default|GetGC_GroundPound) (Ability|GameplayCue|MakeGameplayCueParameters :NormalizedMagnitude 0.0 :RawMagnitude 0.0 :Location (Variables|Default|GetPoundVFXSpawnLocation) :Normal "0,0,0" :GameplayEffectLevel 1 :AbilityLevel 1 :bReplicateLocationWhenUsingMinimalRepProxy false))
            (CallFunction|SpawnRockProjectiles))
          (:Targeted
            (CallFunction|ApplyDamageandKnockback _targetdata _applicationtag)))
        (else
          (Default|SetAllowLookatTarget(Message) (Ability|GetAvatarActorfromActorInfo) true)))))
  (Ability|EndAbility))

(event Ability|EventOnEndAbility
  (params (bWasCancelled :type Boolean))
  (Default|SetAllowLookatTarget(Message) (Variables|InternalData|GetBossActor) true))

(event Custom|ApplyDamageAndKnockback
  (params (TargetData :type Gameplay%20Ability%20Target%20Data%20Handle%20Structure) (ApplicationTag :type Gameplay%20Tag%20Structure))
  ; Apply Damage and Knockback
  (Ability|ApplyGameplayEffectContainer TargetData ApplicationTag)
  (for _array_element (Ability|TargetData|GetAllActorsfromTargetData TargetData)
    (bind _asfort_pawn (Utilities|Casting|CastToFortPawn _array_element)
      (:then
        (CallFunction|KnockbackPawns _asfort_pawn))
      (:CastFailed
        (bind _asfort_athena_vehicle (Utilities|Casting|CastToFortAthenaVehicle _array_element))
        (CallFunction|KnockbackVehicles _asfort_athena_vehicle)))))

(event Custom|CalculateRockFallingPositions
  ; Find Locations to Spawn Rock Projectiles ad Spawn a Tell to let player sknow that something will be happening here soon.
  (Variables|InternalData|SetProjectilesToSpawn (Math|Float|Truncate (ScalableFloat|GetValueatLevel (Variables|Default|Projectiles|GetProjectileCount))))
  (Utilities|Array|Clear (Variables|InternalData|GetRockSpawnLocations))
  (bind _returnvalue (AI|EQS|RunEQSQuery (Variables|Default|Projectiles|GetProjectileSpawnLocationEQS) (Variables|InternalData|GetBossActor) "AllMatching"))
  (AI|EQS|BindEventtoOnQueryFinished _returnvalue (delegate Custom|ProjectileGroundSpawnLocations)))

(event Custom|ProjectileGroundSpawnLocations
  (params (QueryInstance :type EQS%20Query%20Instance%20Object%20Reference) (QueryStatus :type EEnvQueryStatus%20Enum :default "Processing"))
  ; Find Locations to Spawn Rock Projectiles ad Spawn a Tell to let player sknow that something will be happening here soon.
  (bind (_resultlocations _returnvalue) (AI|EQS|GetQueryResultsasLocations QueryInstance))
  (if _returnvalue
    (bind _output_get (Variables|InternalData|SetEQSLocations _resultlocations))
    (Utilities|Array|Shuffle _output_get)
    (if (>= (Utilities|Array|Length _resultlocations) (Variables|InternalData|GetProjectilesToSpawn))
      (Variables|InternalData|SetProjectilesFound 0)
      (bind _array_element (Utilities|Array|ForEachLoopwithBreak (Variables|InternalData|GetEQSLocations))
        (:LoopBody
          (bind (_outhit _returnvalue_2) (Collision|LineTraceByChannel :Start (+ _array_element "0.000000,0.000000,25.000000") :End (+ _array_element "0.000000,0.000000,8000.000000") :TraceChannel "TraceTypeQuery10" :bTraceComplex false :DrawDebugType "ForDuration" :bIgnoreSelf true :TraceColor "(R=1.000000,G=0.000000,B=0.000000,A=1.000000)" :TraceHitColor "(R=0.000000,G=1.000000,B=0.000000,A=1.000000)" :DrawTime 2e+01))
          (if _returnvalue_2
            (if (>= (Math|Vector|Distance(Vector) (.ImpactPoint (Collision|BreakHitResult _outhit)) _array_element) (ScalableFloat|GetValueatLevel (Variables|Default|Projectiles|GetMinDistanceForProjectileTrace)))
              (Utilities|Array|Add (Variables|InternalData|GetRockSpawnLocations) (- (.ImpactPoint (Collision|BreakHitResult _outhit)) "0.000000,0.000000,100.000000"))
              (bind _returnvalue_3 (Game|SpawnActorfromClass "/TidalCrane_Swarm_Boss/Blueprints/BGA_TidalCrane_Boss_FallingRockLocation.BGA_TidalCrane_Boss_FallingRockLocation_C" (Math|Transform|MakeTransform (.ImpactPoint (Collision|BreakHitResult _outhit))) "AlwaysSpawn" "MultiplyWithRoot" 0 0 0))
              (Class|BGATidalCraneBossFallingRockLocation|SetGroundTraceLocation _returnvalue_3 _array_element :!class /TidalCrane_Swarm_Boss/Blueprints/BGA_TidalCrane_Boss_FallingRockLocation.BGA_TidalCrane_Boss_FallingRockLocation_C)
              (bind _result (Math|Integer|IncrementInt (Variables|InternalData|GetProjectilesFound))
                (:%20%20
                  (if (>= _result (Variables|InternalData|GetProjectilesToSpawn))
                    _)))))))
      (else
        _))
    (else
      _))
  (Ability|EndAbility))

(event Custom|SpawnRockProjectiles
  ; Spawn Falling Rock Projectiles
  (for _array_element (Variables|InternalData|GetRockSpawnLocations)
    (Projectile|SpawnProjectile :ProjectileClass (Variables|Default|Projectiles|GetProjectile%20Class) :RequestedBy (Ability|GetAvatarActorfromActorInfo) :SpawnLocation _array_element :SpawnRotation "0, 0, 0" :SpawnDirection (Math|Rotator|MakeRotator 0.0 -9e+01) :EffectContainerSpecToApplyOnExplode (Ability|MakeGameplayEffectContainerSpec "(TagName=\"Targeting.ApplicationTag.1\")") :bUseDefaultPhysics false :InitialSpeed (ScalableFloat|GetValueatLevel (Variables|Default|Projectiles|GetProjectileSpeed)) :GravityScale 1.0 :ChargePercent 0.0)))

(fn GetCustomAbilitySourceTransform
  (return (Math|Transform|MakeTransform (+ (+ (Transformation|GetActorLocation (Ability|GetAvatarActorfromActorInfo)) (* (Transformation|GetActorForwardVector (Ability|GetAvatarActorfromActorInfo)) 7e+02)) "0.000000,0.000000,150.000000") (Transformation|GetActorRotation (Ability|GetAvatarActorfromActorInfo)))))

(fn CanActivateAbility
  (params (ActorInfo :type Gameplay%20Ability%20Actor%20Info%20Structure) (Handle :type Gameplay%20Ability%20Spec%20Handle%20Structure))
  (bind _returnvalue (Ability|GetAvatarActorfromActorInfo))
  (bind _currenttargetactor (Default|GetCurrentTargetActor(Message) _returnvalue))
  (Utilities|IsValid _currenttargetactor
    (:Is%20Valid
      (if (ScalableFloat|GetValueasBool (Variables|Default|Projectiles|GetCanSpawnRockProjectiles))
        (return 0 true)
        (else
          (return 0 (<= (Math|Vector|Distance(Vector) (Transformation|GetActorLocation _returnvalue) (Transformation|GetActorLocation _currenttargetactor)) (ScalableFloat|GetValueatLevel (Variables|Default|GetMaximumDistanceForAttack)))))))
    (:Is%20Not%20Valid
      (return 0 false))))

(fn KnockbackPawns
  (params (Pawn :type Fort%20Pawn%20Object%20Reference))
  (bind (_returnvalue_x _returnvalue_y) (Math|Vector|BreakVector (Math|Vector|Normalize (- (Transformation|GetActorLocation Pawn) (Transformation|GetActorLocation (Variables|InternalData|GetBossActor))))))
  (Character|LaunchCharacter Pawn (Math|Vector|MakeVector (* _returnvalue_x (ScalableFloat|GetValueatLevel (Variables|Default|GetKnockbackPawnForwardVelocity))) (* _returnvalue_y (ScalableFloat|GetValueatLevel (Variables|Default|GetKnockbackPawnForwardVelocity))) (ScalableFloat|GetValueatLevel (Variables|Default|GetKnockbackPawnHeightVelocity))) true true))

(fn KnockbackVehicles
  (params (FortAthenaVehicle :type Fort%20Athena%20Vehicle%20Object%20Reference))
  (bind (_returnvalue_x _returnvalue_y) (Math|Vector|BreakVector (Math|Vector|Normalize (- (Transformation|GetActorLocation FortAthenaVehicle) (Transformation|GetActorLocation (Variables|InternalData|GetBossActor))))))
  (Audio|BroadcastAbilityImpactatLocation 0 0 (Math|Vector|MakeVector (* _returnvalue_x (ScalableFloat|GetValueatLevel (Variables|Default|GetKnockbackVehicleForwardVelocity))) (* _returnvalue_y (ScalableFloat|GetValueatLevel (Variables|Default|GetKnockbackVehicleForwardVelocity))) (ScalableFloat|GetValueatLevel (Variables|Default|GetKnockbackVehicleHeightVelocity))) 777.0 "None" false))
