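; Gameplay ability that makes the Tidal Crane swarm boss pick a new target.
; It can activate only when the avatar's BPI TidalCraneBoss interface reports more than one potential target.
; On activation it commits, waits a brief delay, calls HandleNewTargetSelection on the avatar through the interface, then ends, leaving the selection logic to the boss.
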(event Ability|EventActivateAbility
  (Ability|CommitAbility)
  (Utilities|FlowControl|Delay 0.25)
  (Class|BPITidalCraneBoss|HandleNewTargetSelection(Message) (Ability|GetAvatarActorfromActorInfo))
  (Ability|EndAbility))

(fn CanActivateAbility
  (params (ActorInfo :type Gameplay%20Ability%20Actor%20Info%20Structure) (Handle :type Gameplay%20Ability%20Spec%20Handle%20Structure))
  (bind _numberofpotentialtargets (Class|BPITidalCraneBoss|GetNumberOfPotentialTargets(Message) (Ability|GetAvatarActorfromActorInfo)))
  (return 0 (> _numberofpotentialtargets 1)))
