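; Boss ability for the Tidal Crane boss that fires a projectile volley on montage notifies, using a randomly chosen volley type.
; Volleys target random EQS points, nearby pawns (another type if too few), spots around the target's velocity-led position to discourage retreat, or rapid-fire single shots at that position.
; All feed SpawnProjectiles, which launches arced or straight/homing projectiles, spawns target-location markers and plays one throw cue per volley.
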(event Ability|EventActivateAbility
  (bind (_montageinfo_montagetoplay _montageinfo_animplayrate _montageinfo_animrootmotiontranslationscale _montageinfo_montagesectiontoplay _montageinfo_overridesection) (Utilities|Struct|BreakFortGameplayAbilityMontageInfo (Variables|GetMontageInfo_MontageToPlay)))
  (Ability|CommitAbility)
  (Utilities|FlowControl|Delay 1.0)
  (Variables|InternalData|SetBossBGA (Ability|GetAvatarActorfromActorInfo))
  (Variables|InternalData|SetRapidFireProjectileCountFired 0)
  (Variables|InternalData|SetChosenProjectileType (CallFunction|ChooseProjectileVolleyType))
  (Variables|InternalData|SetProjectileThrownGCPlayed false)
  (bind (_asynctaskproxy _notifytag) (Ability|Tasks|PlayMontageAndWaitNotify "None" _montageinfo_montagetoplay _montageinfo_animplayrate _montageinfo_overridesection true _montageinfo_animrootmotiontranslationscale)
    (:OnCompleted
      _)
    (:OnInterrupted
      _)
    (:OnCancelled
      _)
    (:OnNotifyBegin
      (if (== _notifytag (Variables|Default|GetTag_BossShootProjectile))
        (CallFunction|SelectVolleyType)
        (elif (Utilities|Enum|Equal(Enum) (Variables|InternalData|GetChosenProjectileType) "NewEnumerator3")
          (if (not (<= (Variables|InternalData|GetRapidFireProjectileCountFired) (- (ScalableFloat|GetValueatLevel (Variables|Default|GetProjectileCount)) 1.0)))
            (Ability|Animation|MontageJumptoSection "Outro"))
          (else
            _)))))
  (Ability|EndAbility))

(event Custom|SelectVolleyType
  ; Select What Type of Volley to use
  (Utilities|Array|Clear (Variables|InternalData|GetVolleyTargetLocations))
  (switch Utilities|FlowControl|Switch|SwitchonENUM_TidalCrane_Boss_VolleyType (Variables|InternalData|GetChosenProjectileType)
    (:NewEnumerator0
      (CallFunction|FindRandomProjectileTargetLocations))
    (:NewEnumerator1
      (bind _allnearbypawns (Default|GetAllNearbyPawns(Message) (Variables|InternalData|GetBossBGA)))
      (if (> (Utilities|Array|Length _allnearbypawns) 1)
        (CallFunction|FindNearTargetsLocations)
        (else
          (switch int (Math|Random|RandomIntegerinRange 0 2)
            (:0
              (Variables|InternalData|SetChosenProjectileType "NewEnumerator0")
              (CallFunction|FindRandomProjectileTargetLocations))
            (:1
              (Variables|InternalData|SetChosenProjectileType "NewEnumerator2")
              (CallFunction|BehindMainTarget))
            (:2
              (Variables|InternalData|SetChosenProjectileType "NewEnumerator3")
              (CallFunction|RapidFire))))))
    (:NewEnumerator2
      (CallFunction|BehindMainTarget))
    (:NewEnumerator3
      (CallFunction|RapidFire))))

(event Custom|FindRandomProjectileTargetLocations
  ; Using EQS, Find RAndom Locations and then Spawn Projectiles are those locations.
  (bind _returnvalue (AI|EQS|RunEQSQuery (Variables|Default|GetEQS_RadomLocations) (Variables|InternalData|GetBossBGA) "AllMatching"))
  (AI|EQS|BindEventtoOnQueryFinished _returnvalue (delegate Custom|RandomLocationEQSFinished)))

(event Custom|RandomLocationEQSFinished
  (params (QueryInstance :type EQS%20Query%20Instance%20Object%20Reference) (QueryStatus :type EEnvQueryStatus%20Enum :default "Processing"))
  (bind _randomeqslocations (Variables|InternalData|GetRandomEQSLocations))
  (bind (_outitem _outindex) (Utilities|Array|RandomArrayItem _randomeqslocations))
  ; Using EQS, Find RAndom Locations and then Spawn Projectiles are those locations.
  (bind (_resultlocations _returnvalue) (AI|EQS|GetQueryResultsasLocations QueryInstance))
  (if _returnvalue
    (Variables|InternalData|SetRandomEQSLocations _resultlocations)
    (Utilities|Array|Clear (Variables|InternalData|GetVolleyTargetLocations))
    (for _ (range (+ (Math|Float|Truncate (- (ScalableFloat|GetValueatLevel (Variables|Default|GetProjectileCount)) 1.0)) 1))
      (bind _output_get (Variables|InternalData|SetTempArrayIndex _outindex))
      (Utilities|Array|AddUnique (Variables|InternalData|GetVolleyTargetLocations) (Utilities|Array|Get(acopy) _randomeqslocations _output_get))
      (Utilities|Array|RemoveIndex (Variables|InternalData|GetRandomEQSLocations) (Variables|InternalData|GetTempArrayIndex)))
    (CallFunction|SpawnProjectiles)
    (else
      (Ability|EndAbility))))

(event Custom|FindNearTargetsLocations
  (bind _tempnearbytargetactors (Variables|InternalData|GetTempNearbyTargetActors))
  (bind (_outitem _outindex) (Utilities|Array|RandomArrayItem _tempnearbytargetactors))
  (bind _projectilecount (Variables|Default|GetProjectileCount))
  (bind _returnvalue (ScalableFloat|GetValueatLevel _projectilecount))
  ; Get At most [Projectile Count] Targets and throw a Projectile their way (ignoring Velocity on Purpose)
  (bind _allnearbypawns (Default|GetAllNearbyPawns(Message) (Variables|InternalData|GetBossBGA)))
  (bind _output_get (Variables|InternalData|SetUseProjectileCountForNearbyTargets (> (Utilities|Array|Length _allnearbypawns) _returnvalue)))
  (if _output_get
    (Variables|InternalData|SetTempNearbyTargetActors _allnearbypawns)
    (for _ (range (+ (Math|Float|Truncate (- _returnvalue 1.0)) 1))
      (bind _output_get_2 (Variables|InternalData|SetTempArrayIndex _outindex))
      (Utilities|Array|Add (Variables|InternalData|GetNearbyTargetActors) (Utilities|Array|Get(acopy) _tempnearbytargetactors _output_get_2))
      (Utilities|Array|RemoveIndex _tempnearbytargetactors _output_get_2))
    (else
      (Variables|InternalData|SetNearbyTargetActors _allnearbypawns)))
  (for _array_element (Variables|InternalData|GetNearbyTargetActors)
    (Utilities|Array|Add (Variables|InternalData|GetVolleyTargetLocations) (Transformation|GetActorLocation _array_element)))
  (CallFunction|SpawnProjectiles))

(event Custom|BehindMainTarget
  (bind _bossbga (Variables|InternalData|GetBossBGA))
  ; Put pressure on the player to try to move forwars instead of backewards by throwing projectiles behind / next to the player.
  (bind _currenttargetactor (Default|GetCurrentTargetActor(Message) _bossbga))
  (Utilities|IsValid _currenttargetactor
    (:Is%20Valid
      (bind _output_get (Variables|InternalData|SetTargetLocation (+ (Transformation|GetActorLocation _currenttargetactor) (Transformation|GetVelocity _currenttargetactor))))
      (Variables|InternalData|SetBossDirection (Math|Vector|Normalize (- (Transformation|GetActorLocation _bossbga) _output_get)))
      (for _array_element (Variables|Default|GetDegreeValues)
        (Utilities|Array|Add (Variables|InternalData|GetVolleyTargetLocations) (+ (Variables|InternalData|GetTargetLocation) (* (Math|Vector|RotateVectorAroundAxis (Variables|InternalData|GetBossDirection) (Math|Conversions|ToFloat(Integer) _array_element) "0.000000,0.000000,1.000000") (ScalableFloat|GetValueatLevel (Variables|Default|GetProjectileSpreadDistance))))))
      (CallFunction|SpawnProjectiles))
    (:Is%20Not%20Valid
      (Ability|EndAbility))))

(event Custom|SpawnProjectiles
  (bind _location (CallFunction|GetProjectileSpawnLocation))
  (bind _returnvalue (Ability|MakeGameplayEffectContainerSpec "(TagName=\"Targeting.ApplicationTag.0\")"))
  (bind _returnvalue_2 (Ability|GetAvatarActorfromActorInfo))
  (bind (_outprojectilespeed _outspawndirection _outprojectilegravity _homingtarget) (Ability|GetProjectileInitialValues true))
  (bind _volleytargetlocations (Variables|InternalData|GetVolleyTargetLocations))
  (bind _projectileclass (Variables|Ability|GetProjectileClass))
  ; Spawn Projectiles
  (Class|BPITidalCraneBoss|SetProjectileVolleySpawnLocations(Message) (Variables|InternalData|GetBossBGA) _volleytargetlocations)
  (for _array_element _volleytargetlocations
    (bind _output_get (Variables|InternalData|SetTargetLocation _array_element))
    (if (ScalableFloat|GetValueasBool (Variables|Default|GetUseArcTypeProjectile))
      (bind _outlaunchvelocity (Game|SuggestProjectileVelocityCustomArc (CallFunction|GetProjectileSpawnLocation) _output_get (ScalableFloat|GetValueatLevel (Variables|Default|GetProjectilveGravityOverride)) (ScalableFloat|GetValueatLevel (Variables|Default|GetProjectileArcValue))))
      (bind _returnvalue_3 (Projectile|SpawnProjectile :ProjectileClass _projectileclass :RequestedBy _returnvalue_2 :SpawnLocation _location :SpawnRotation "0, 0, 0" :SpawnDirection (Math|Conversions|RotationFromXVector _outlaunchvelocity) :EffectContainerSpecToApplyOnExplode _returnvalue :bUseDefaultPhysics false :InitialSpeed (Math|Vector|VectorLength _outlaunchvelocity) :GravityScale (ScalableFloat|GetValueatLevel (Variables|Default|GetProjectilveGravityScale)) :ChargePercent 0.0))
      (bind _asprj_tidal_crane_medium_ranged (Utilities|Casting|CastToPrj_TidalCrane_Medium_Ranged _returnvalue_3))
      (Variables|InternalData|SetRangedProjectile _asprj_tidal_crane_medium_ranged)
      (else
        (bind _returnvalue_4 (Projectile|SpawnProjectile :ProjectileClass _projectileclass :RequestedBy _returnvalue_2 :SpawnLocation _location :SpawnRotation _outspawndirection :SpawnDirection _outspawndirection :EffectContainerSpecToApplyOnExplode _returnvalue :bUseDefaultPhysics false :HomingTarget _homingtarget :InitialSpeed _outprojectilespeed :GravityScale _outprojectilegravity :ChargePercent 0.0))
        (bind _asprj_tidal_crane_medium_ranged_2 (Utilities|Casting|CastToPrj_TidalCrane_Medium_Ranged _returnvalue_4))
        (Variables|InternalData|SetRangedProjectile _asprj_tidal_crane_medium_ranged_2)))
    (bind _returnvalue_5 (Game|SpawnActorfromClass "/TidalCrane_Swarm_Boss/Blueprints/BGA_TidalCrane_Boss_ProjectileLocation.BGA_TidalCrane_Boss_ProjectileLocation_C" (Math|Transform|MakeTransform (Variables|InternalData|GetTargetLocation)) "AlwaysSpawn" "MultiplyWithRoot" 0 0 0 0))
    (Variables|/ProtoSwarm/Abilities/Medium/Prj_TidalCrane_Medium_Ranged.Prj_TidalCrane_Medium_Ranged_C|SetTargetProjection _returnvalue_5 (Variables|InternalData|GetRangedProjectile))
    (if (not (Variables|InternalData|GetProjectileThrownGCPlayed))
      (Variables|InternalData|SetProjectileThrownGCPlayed true)
      (Ability|ExecuteGameplayCueOnOwner (Variables|Default|GetGC_ProjectileThrown)))))

(event Custom|RapidFire
  (bind _volleytargetlocations (Variables|InternalData|GetVolleyTargetLocations))
  ; Rapid Fire Projectile Spawn
  (Math|Integer|IncrementInt (Variables|InternalData|GetRapidFireProjectileCountFired)
    (:%20%20
      (Variables|InternalData|SetProjectileThrownGCPlayed false)
      (Utilities|Array|Clear _volleytargetlocations)
      (bind _currenttargetactor (Default|GetCurrentTargetActor(Message) (Variables|InternalData|GetBossBGA)))
      (Utilities|Array|Add _volleytargetlocations (+ (Transformation|GetActorLocation _currenttargetactor) (Transformation|GetVelocity _currenttargetactor)))
      (CallFunction|SpawnProjectiles))))

(fn ChooseProjectileVolleyType
  (return (Utilities|Select :Index (Math|Random|RandomIntegerinRange 0 (- (Utilities|Enum|GetnumberofentriesinENUM_TidalCrane_Boss_VolleyType) 1)) :Option%200 "NewEnumerator0" :Option%201 "NewEnumerator1" :Option%202 "NewEnumerator2" :Option%203 "NewEnumerator3")))

(fn GetProjectileSpawnLocation
  (return (+ (Utilities|Operators|Add (Transformation|GetActorLocation (Variables|InternalData|GetBossBGA)) (Math|Vector|MakeVector 0.0 0.0 (ScalableFloat|GetValueatLevel (Variables|Default|GetProjectileSpawnLocationOffsetZ)))) (* (Transformation|GetActorForwardVector (Variables|InternalData|GetBossBGA)) (ScalableFloat|GetValueatLevel (Variables|Default|GetProjectileSpawnLocationOffsetXY))))))

(fn CanActivateAbility
  (params (ActorInfo :type Gameplay%20Ability%20Actor%20Info%20Structure) (Handle :type Gameplay%20Ability%20Spec%20Handle%20Structure))
  (bind _currenttargetactor (Default|GetCurrentTargetActor(Message) (Ability|GetAvatarActorfromActorInfo)))
  (return 0 (Utilities|IsValid _currenttargetactor :!class /Script/Engine.KismetSystemLibrary)))
