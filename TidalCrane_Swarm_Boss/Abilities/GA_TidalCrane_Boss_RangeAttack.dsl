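; Boss ranged-attack ability for the TidalCrane swarm boss, activatable only when the current target is within a min/max distance band.
; It commits, briefly delays, caches the target, then plays a montage whose notify launches the projectile.
; The projectile uses either a suggested arc velocity toward the target's velocity-led position or standard initial values with homing.
; It also spawns a target-location marker actor, links it to the projectile, and fires a throw GameplayCue.
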(event Ability|EventActivateAbility
  (bind (_montageinfo_montagetoplay _montageinfo_animplayrate _montageinfo_animrootmotiontranslationscale _montageinfo_montagesectiontoplay _montageinfo_overridesection) (Utilities|Struct|BreakFortGameplayAbilityMontageInfo (Variables|GetMontageInfo_MontageToPlay)))
  (CallFunction|PerformCommit)
  (Utilities|FlowControl|Delay 1.0)
  (bind _currenttargetactor (Default|GetCurrentTargetActor(Message) (Ability|GetAvatarActorfromActorInfo)))
  (bind _output_get (Variables|InternalData|SetTargetActor _currenttargetactor))
  (Utilities|IsValid _output_get
    (:Is%20Valid
      (Ability|Tasks|PlayMontageAndWaitNotify "None" _montageinfo_montagetoplay _montageinfo_animplayrate _montageinfo_overridesection true _montageinfo_animrootmotiontranslationscale
        (:OnCompleted
          _)
        (:OnInterrupted
          _)
        (:OnCancelled
          _)
        (:OnNotifyBegin
          (CallFunction|LaunchProjectile)))
      (Ability|EndAbility))
    (:Is%20Not%20Valid
      (Ability|EndAbility))))

(event Custom|PerformCommit
  (Ability|CommitAbility))

(event Custom|LaunchProjectile
  (bind _location (CallFunction|GetProjectileSpawnLocation))
  (bind _returnvalue (Ability|MakeGameplayEffectContainerSpec "(TagName=\"Targeting.ApplicationTag.0\")"))
  (bind _returnvalue_2 (Ability|GetAvatarActorfromActorInfo))
  (bind (_outprojectilespeed _outspawndirection _outprojectilegravity _homingtarget) (Ability|GetProjectileInitialValues true))
  (bind _projectileclass (Variables|Ability|GetProjectileClass))
  ; Spawn Projectile
  (bind _output_get (Variables|InternalData|SetTargetLocation (CallFunction|GetProjectileTargetLocation)))
  (if (ScalableFloat|GetValueasBool (Variables|Default|GetUseArcTypeProjectile))
    (bind _outlaunchvelocity (CallFunction|GetSuggestedProjectileVelocity _output_get (CallFunction|GetProjectileSpawnLocation) (ScalableFloat|GetValueatLevel (Variables|Default|GetArcParam)) (ScalableFloat|GetValueatLevel (Variables|Default|GetProjectilveGravityOverride))))
    (bind _returnvalue_3 (Projectile|SpawnProjectile :ProjectileClass _projectileclass :RequestedBy _returnvalue_2 :SpawnLocation _location :SpawnRotation "0, 0, 0" :SpawnDirection (Math|Conversions|RotationFromXVector _outlaunchvelocity) :EffectContainerSpecToApplyOnExplode _returnvalue :bUseDefaultPhysics false :InitialSpeed (Math|Vector|VectorLength _outlaunchvelocity) :GravityScale (ScalableFloat|GetValueatLevel (Variables|Default|GetProjectilveGravityScale)) :ChargePercent 0.0))
    (else
      (Projectile|SpawnProjectile :ProjectileClass _projectileclass :RequestedBy _returnvalue_2 :SpawnLocation _location :SpawnRotation _outspawndirection :SpawnDirection _outspawndirection :EffectContainerSpecToApplyOnExplode _returnvalue :bUseDefaultPhysics false :HomingTarget _homingtarget :InitialSpeed _outprojectilespeed :GravityScale _outprojectilegravity :ChargePercent 0.0)))
  (bind _asprj_tidal_crane_medium_ranged (Utilities|Casting|CastToPrj_TidalCrane_Medium_Ranged _returnvalue_3))
  (Variables|Default|SetRangedProjectile _asprj_tidal_crane_medium_ranged)
  (bind _returnvalue_4 (Game|SpawnActorfromClass "/TidalCrane_Swarm_Boss/Blueprints/BGA_TidalCrane_Boss_ProjectileLocation.BGA_TidalCrane_Boss_ProjectileLocation_C" (Math|Transform|MakeTransform (Variables|InternalData|GetTargetLocation)) "AlwaysSpawn" "MultiplyWithRoot" 0 _returnvalue_2 0 (Ability|GetActivatingPawn)))
  (Variables|/ProtoSwarm/Abilities/Medium/Prj_TidalCrane_Medium_Ranged.Prj_TidalCrane_Medium_Ranged_C|SetTargetProjection _returnvalue_4 (Variables|Default|GetRangedProjectile))
  (Ability|ExecuteGameplayCueOnOwner (Variables|Default|GetGC_ProjectileThrown)))

(fn GetProjectileSpawnLocation
  (return (+ (Utilities|Operators|Add (Transformation|GetActorLocation (Ability|GetAvatarActorfromActorInfo)) (Math|Vector|MakeVector 0.0 0.0 (ScalableFloat|GetValueatLevel (Variables|Default|GetProjectileSpawnLocationOffsetZ)))) (* (Transformation|GetActorForwardVector (Ability|GetAvatarActorfromActorInfo)) (ScalableFloat|GetValueatLevel (Variables|Default|GetProjectileSpawnLocationOffsetXY))))))

(fn GetProjectileTargetLocation
  (return (+ (Transformation|GetActorLocation (Variables|InternalData|GetTargetActor)) (Transformation|GetVelocity (Variables|InternalData|GetTargetActor)))))

(fn GetSuggestedProjectileVelocity
  (params (EndPos :type Vector) (StartPos :type Vector) (ArcParam :type Float%20%28single-precision%29) (OverrideGravityZ :type Float%20%28single-precision%29))
  (bind _outlaunchvelocity (Game|SuggestProjectileVelocityCustomArc StartPos EndPos OverrideGravityZ ArcParam))
  (return _outlaunchvelocity))

(fn CanActivateAbility
  (params (ActorInfo :type Gameplay%20Ability%20Actor%20Info%20Structure) (Handle :type Gameplay%20Ability%20Spec%20Handle%20Structure))
  (bind _returnvalue (Ability|GetAvatarActorfromActorInfo))
  (bind _currenttargetactor (Default|GetCurrentTargetActor(Message) _returnvalue))
  (Utilities|IsValid _currenttargetactor
    (:Is%20Valid
      (return 0 (and (<= (Math|Vector|Distance(Vector) (Transformation|GetActorLocation _returnvalue) (Transformation|GetActorLocation _currenttargetactor)) (ScalableFloat|GetValueatLevel (Variables|Default|GetMaximumDistanceForAttack))) (>= (Math|Vector|Distance(Vector) (Transformation|GetActorLocation _returnvalue) (Transformation|GetActorLocation _currenttargetactor)) (ScalableFloat|GetValueatLevel (Variables|Default|GetMinimumDistanceForAttack))))))
    (:Is%20Not%20Valid
      (return 0 false))))
