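; Gameplay ability that repositions the TidalCrane swarm boss by burrowing it to a new spot and surfacing it again.
; It picks the reposition actor nearest the target, excluding the one nearest the boss, or falls back to a random EQS result.
; Burrow montage notifies drive the move, then either a delay or timed, landscape-traced damage-effect pulses along the path.
; Surfacing attempts knockback, plays a separate unburrow montage to avoid premature ending, and spawns a crater.
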(event Ability|EventActivateAbility
  (if (Variables|Default|GetbCommitAbilityOnActivate)
    (Ability|CommitAbility)
    (else
      _))
  (Variables|InternalData|SetBossDistance -1.0)
  (Variables|InternalData|SetPlayerDistance -1.0)
  (bind _output_get (Variables|InternalData|SetBossActor (Ability|GetAvatarActorfromActorInfo)))
  (bind _currenttargetactor (Default|GetCurrentTargetActor(Message) _output_get))
  (Variables|InternalData|SetCurrentTargetActor _currenttargetactor)
  (CallFunction|BeginRepositionQuery))

(event Custom|BeginRepositionQuery
  (bind _repositionactors (Default|GetRepositionActors(Message) (Variables|InternalData|GetBossActor)))
  (bind _output_get (Variables|Default|SetReposition%20Actors _repositionactors))
  (if (> (Utilities|Array|Length _output_get) 0)
    (CallFunction|RelocateUsingActors)
    (else
      (CallFunction|RelocateUsingEQS))))

(event Custom|RelocateUsingEQS
  ; Find Sawn Locations around the Player
  (bind _returnvalue (AI|EQS|RunEQSQuery (Variables|Default|GetEQS_Repositioning) (Variables|InternalData|GetCurrentTargetActor) (Variables|Default|GetEnvQueryRunMode)))
  (AI|EQS|BindEventtoOnQueryFinished _returnvalue (EventDispatchers|CreateEvent :!event Custom|RepositioningEQSFinished)))

(event Custom|RepositioningEQSFinished
  (params (QueryInstance :type EQS%20Query%20Instance%20Object%20Reference) (QueryStatus :type EEnvQueryStatus%20Enum :default "Processing"))
  ; Find Sawn Locations around the Player
  (if (Utilities|Enum|Equal(Enum) QueryStatus "Success")
    (bind (_resultlocations _returnvalue) (AI|EQS|GetQueryResultsasLocations QueryInstance))
    (if _returnvalue
      (Variables|Default|SetNewLocation (Utilities|Array|RandomArrayItem _resultlocations))
      (CallFunction|StartBurrowSequence true)
      (else
        _))
    (else
      _))
  (Ability|EndAbility))

(event Custom|RelocateUsingActors
  (bind _bossactor (Variables|InternalData|GetBossActor))
  (bind _returnvalue (Transformation|GetActorLocation _bossactor))
  (bind _currenttargetactor (Variables|InternalData|GetCurrentTargetActor))
  (bind _returnvalue_2 (Transformation|GetActorLocation _currenttargetactor))
  ; Find one of the Burrow Locations that is not next to the Boss, but is the closest to the Player
  (for (_array_element _array_index) (Variables|Default|GetReposition%20Actors)
    (if (or (< (Math|Vector|Distance(Vector) (Transformation|GetActorLocation _array_element) _returnvalue) (Variables|InternalData|GetBossDistance)) (< (Variables|InternalData|GetBossDistance) 0))
      (Variables|InternalData|SetBossDistance (Math|Vector|Distance(Vector) (Transformation|GetActorLocation _array_element) _returnvalue))
      (Variables|InternalData|SetBossDistanceIndex _array_index)))
  (Utilities|Array|RemoveIndex (Variables|Default|GetReposition%20Actors) (Variables|InternalData|GetBossDistanceIndex)) ; Remove the current Entry that's closest to the boss
  (Variables|Default|SetFoundValidRepositionLocation false)
  (for _array_element_2 (Variables|Default|GetReposition%20Actors)
    (if (or (< (Math|Vector|Distance(Vector) (Transformation|GetActorLocation _array_element_2) _returnvalue_2) (Variables|InternalData|GetPlayerDistance)) (< (Variables|InternalData|GetPlayerDistance) 0))
      (Variables|InternalData|SetPlayerDistance (Math|Vector|Distance(Vector) (Transformation|GetActorLocation _array_element_2) _returnvalue_2))
      (Variables|Default|SetNewLocation (Transformation|GetActorLocation _array_element_2))
      (Variables|Default|SetFoundValidRepositionLocation true)))
  (if (Variables|Default|GetFoundValidRepositionLocation)
    (CallFunction|StartBurrowSequence true)
    (else
      (CallFunction|RelocateUsingEQS))))

(event Custom|StartBurrowSequence
  (params (bAutoMoveToNewLocation :type Boolean :default true))
  (bind (_montageinfo_montagetoplay _montageinfo_animplayrate _montageinfo_animrootmotiontranslationscale) (Utilities|Struct|BreakFortGameplayAbilityMontageInfo (Variables|GetMontageInfo_MontageToPlay)))
  ; Burrow and Unburrow
  (bind (_asynctaskproxy _notifytag) (Ability|Tasks|PlayMontageAndWaitNotify "None" _montageinfo_montagetoplay _montageinfo_animplayrate "None" true _montageinfo_animrootmotiontranslationscale)
    (:OnNotifyBegin
      (Utilities|FlowControl|Sequence
        (:then_0
          (CallFunction|NotifyBegin _notifytag))
        (:then_1
          (if (== _notifytag (Variables|Default|GetTag_Burrowing))
            (Ability|ExecuteGameplayCueOnOwner (Variables|Default|GetGC_Burrow))
            (Variables|InternalData|SetPreRelocationLocation (Transformation|GetActorLocation (Variables|InternalData|GetBossActor)))
            (if (not (Variables|BurrowTravelGameplay|GetbExecuteBurrowTravelGameplay))
              (Utilities|FlowControl|Delay 1.0)
              (CallFunction|SpawnDustTrail (Variables|InternalData|GetPreRelocationLocation) (Variables|Default|GetNewLocation)))
            (elif (== _notifytag (Variables|Default|GetTag_BurrowingComplete))
              (if bAutoMoveToNewLocation
                (Default|HandleMovetoNewLocation(Message) (Variables|InternalData|GetBossActor) (Variables|Default|GetNewLocation))
                (if (Variables|BurrowTravelGameplay|GetbExecuteBurrowTravelGameplay)
                  (CallFunction|StartBurrowTravelGameplay (Variables|InternalData|GetPreRelocationLocation) (Variables|Default|GetNewLocation))
                  (else
                    (Utilities|FlowControl|Delay (ScalableFloat|GetValueatLevel (Variables|Default|GetBurrowDuration)))
                    (CallFunction|TryKnockbackwithGameplayEffect)
                    ;;[0.04982 0.1451 0.15 0.5] I had originally did a JumpToNextSection in the same montage, but strangely, that could set the MOntage to Complete and end the ABility Pre-maturely in 90% of the time, so to make sure we see the Poke back out animation, I have to call it as a seperate Play Montage :\
                    (bind (_asynctaskproxy_2 _notifytag_2) (Ability|Tasks|PlayMontageAndWaitNotify "None" (Variables|Animation|GetMontageUnburrow))
                      (:then
                        (Ability|ExecuteGameplayCueOnOwner (Variables|Default|GetGC_Unburrow))
                        (Default|HandleSpawnCraterBaseActor(Message) (Variables|InternalData|GetBossActor)))
                      (:OnCompleted
                        _)
                      (:OnInterrupted
                        _)
                      (:OnCancelled
                        _)
                      (:OnNotifyBegin
                        (CallFunction|NotifyBegin _notifytag_2)))
                    (Ability|EndAbility)))))))))))

(event Custom|FullBurrowTravelGameplayComplete
  ; Burrow and Unburrow
  (CallFunction|TryKnockbackwithGameplayEffect)
  ;;[0.04982 0.1451 0.15 0.5] I had originally did a JumpToNextSection in the same montage, but strangely, that could set the MOntage to Complete and end the ABility Pre-maturely in 90% of the time, so to make sure we see the Poke back out animation, I have to call it as a seperate Play Montage :\
  (bind (_asynctaskproxy _notifytag) (Ability|Tasks|PlayMontageAndWaitNotify "None" (Variables|Animation|GetMontageUnburrow))
    (:then
      (Ability|ExecuteGameplayCueOnOwner (Variables|Default|GetGC_Unburrow))
      (Default|HandleSpawnCraterBaseActor(Message) (Variables|InternalData|GetBossActor)))
    (:OnCompleted
      _)
    (:OnInterrupted
      _)
    (:OnCancelled
      _)
    (:OnNotifyBegin
      (CallFunction|NotifyBegin _notifytag)))
  (Ability|EndAbility))

(event Custom|MoveToNewLocation
  ; Burrow and Unburrow
  (Default|HandleMovetoNewLocation(Message) (Variables|InternalData|GetBossActor) (Variables|Default|GetNewLocation))
  (if (Variables|BurrowTravelGameplay|GetbExecuteBurrowTravelGameplay)
    (CallFunction|StartBurrowTravelGameplay (Variables|InternalData|GetPreRelocationLocation) (Variables|Default|GetNewLocation))
    (else
      (Utilities|FlowControl|Delay (ScalableFloat|GetValueatLevel (Variables|Default|GetBurrowDuration)))
      (CallFunction|TryKnockbackwithGameplayEffect)
      ;;[0.04982 0.1451 0.15 0.5] I had originally did a JumpToNextSection in the same montage, but strangely, that could set the MOntage to Complete and end the ABility Pre-maturely in 90% of the time, so to make sure we see the Poke back out animation, I have to call it as a seperate Play Montage :\
      (bind (_asynctaskproxy _notifytag) (Ability|Tasks|PlayMontageAndWaitNotify "None" (Variables|Animation|GetMontageUnburrow))
        (:then
          (Ability|ExecuteGameplayCueOnOwner (Variables|Default|GetGC_Unburrow))
          (Default|HandleSpawnCraterBaseActor(Message) (Variables|InternalData|GetBossActor)))
        (:OnCompleted
          _)
        (:OnInterrupted
          _)
        (:OnCancelled
          _)
        (:OnNotifyBegin
          (CallFunction|NotifyBegin _notifytag)))
      (Ability|EndAbility))))

(event Custom|NotifyBegin
  (params (NotifyTag :type Gameplay%20Tag%20Structure)))

(event Custom|StartBurrowTravelGameplay
  (params (StartLocation :type Vector) (EndLocation :type Vector))
  ;[0 0.15 0.09053 0.5] BurrowTraveGameplay
  ;;[0.01134 0.08569 0.15 0.5] StartBurrowTravelGameplay: Cache the necessary variables and start the burrow travel gameplay process.
  (bind _output_get (Variables|Transient|BurrowTravelGameplay|SetBurrowTravelStartLocation StartLocation))
  (bind _output_get_2 (Variables|Transient|BurrowTravelGameplay|SetBurrowTravelEndLocation EndLocation))
  (Variables|Transient|BurrowTravelGameplay|SetBurrowTravelDirection2D (Math|Vector|Normalize2D(Vector) (Math|Vector|GetUnitDirection(Vector) _output_get _output_get_2)))
  (bind _output_get_3 (Variables|Transient|BurrowTravelGameplay|SetBurrowTravelDistance (Math|Vector|Distance2D(Vector) _output_get_2 _output_get)))
  (Variables|Transient|BurrowTravelGameplay|SetTotalNumberOfBurrowTravelPulses (Math|Float|Floor (Math|Float|SafeDivide _output_get_3 (ScalableFloat|GetCachedScalableFloatValue (Variables|BurrowTravelGameplay|GetHotfix_BurrowTravelGameplayDistanceBetweenPulse)))))
  (Variables|Transient|BurrowTravelGameplay|SetCurrentBurrowTravelPulsesCount 0)
  (bind _returnvalue (Utilities|Time|SetTimerbyEvent (EventDispatchers|CreateEvent :!event Custom|ExecuteNextBurrowTravelPulse) (ScalableFloat|GetCachedScalableFloatValue (Variables|BurrowTravelGameplay|GetHotfix_BurrowTravelGameplayDelayBetweenPulse)) true))
  (Variables|Transient|BurrowTravelGameplay|SetBurrowTravelPulseTimerHandle _returnvalue))

(event Custom|ExecuteNextBurrowTravelPulse
  (bind _currentburrowtravelpulsescount (Variables|Transient|BurrowTravelGameplay|GetCurrentBurrowTravelPulsesCount))
  (bind _burrowtraveldirection2d (Variables|Transient|BurrowTravelGameplay|GetBurrowTravelDirection2D))
  ;[0 0.15 0.09053 0.5] BurrowTraveGameplay
  ;;[0.03089 0.03305 0.15 0.5] ExecuteNextBurrowTravelPulse: Execute the pulse at the next travel location.
  (if (< _currentburrowtravelpulsescount (Variables|Transient|BurrowTravelGameplay|GetTotalNumberOfBurrowTravelPulses))
    (bind _result (Math|Integer|IncrementInt _currentburrowtravelpulsescount)
      (:%20%20
        (bind (_bvalid _hitresult) (BurrowTravelGameplay|FindLandscapeatLocation (+ (Variables|Transient|BurrowTravelGameplay|GetBurrowTravelStartLocation) (Utilities|Operators|Multiply _burrowtraveldirection2d (ScalableFloat|GetCachedScalableFloatValue (Variables|BurrowTravelGameplay|GetHotfix_BurrowTravelGameplayDistanceBetweenPulse)) _result))))
        (if _bvalid
          (Variables|Transient|BurrowTravelGameplay|SetLatestBurrowTravelPulseLocation (.ImpactPoint (Collision|BreakHitResult _hitresult)))
          (bind (_asynctaskproxy _targetdata _applicationtag) (Ability|Tasks|DoTargetSelection "None" (Variables|BurrowTravelGameplay|GetBurrowTravelPulseApplicationTag))
            (:then
              (Ability|ExecuteGameplayCueWithParamsOnOwner (Utilities|Struct|BreakGameplayCueTag (Variables|BurrowTravelGameplay|GetBurrowTravelPulseGameplayCueTag)) (Ability|GameplayCue|MakeGameplayCueParameters (Ability|GameplayCue|BreakGameplayCueParameters (BurrowTravelGameplay|MakeGameplayCueParamfromHitResultforBurrowTravel _hitresult)) (.RawMagnitude (Ability|GameplayCue|BreakGameplayCueParameters (BurrowTravelGameplay|MakeGameplayCueParamfromHitResultforBurrowTravel _hitresult))) (.EffectContext (Ability|GameplayCue|BreakGameplayCueParameters (BurrowTravelGameplay|MakeGameplayCueParamfromHitResultforBurrowTravel _hitresult))) (.MatchedTagName (Ability|GameplayCue|BreakGameplayCueParameters (BurrowTravelGameplay|MakeGameplayCueParamfromHitResultforBurrowTravel _hitresult))) (.OriginalTag (Ability|GameplayCue|BreakGameplayCueParameters (BurrowTravelGameplay|MakeGameplayCueParamfromHitResultforBurrowTravel _hitresult))) (.AggregatedSourceTags (Ability|GameplayCue|BreakGameplayCueParameters (BurrowTravelGameplay|MakeGameplayCueParamfromHitResultforBurrowTravel _hitresult))) (.AggregatedTargetTags (Ability|GameplayCue|BreakGameplayCueParameters (BurrowTravelGameplay|MakeGameplayCueParamfromHitResultforBurrowTravel _hitresult))) (.Location (Ability|GameplayCue|BreakGameplayCueParameters (BurrowTravelGameplay|MakeGameplayCueParamfromHitResultforBurrowTravel _hitresult))) _burrowtraveldirection2d (.Instigator (Ability|GameplayCue|BreakGameplayCueParameters (BurrowTravelGameplay|MakeGameplayCueParamfromHitResultforBurrowTravel _hitresult))) (.EffectCauser (Ability|GameplayCue|BreakGameplayCueParameters (BurrowTravelGameplay|MakeGameplayCueParamfromHitResultforBurrowTravel _hitresult))) (.SourceObject (Ability|GameplayCue|BreakGameplayCueParameters (BurrowTravelGameplay|MakeGameplayCueParamfromHitResultforBurrowTravel _hitresult))) (.PhysicalMaterial (Ability|GameplayCue|BreakGameplayCueParameters (BurrowTravelGameplay|MakeGameplayCueParamfromHitResultforBurrowTravel _hitresult))) (.GameplayEffectLevel (Ability|GameplayCue|BreakGameplayCueParameters (BurrowTravelGameplay|MakeGameplayCueParamfromHitResultforBurrowTravel _hitresult))) (.AbilityLevel (Ability|GameplayCue|BreakGameplayCueParameters (BurrowTravelGameplay|MakeGameplayCueParamfromHitResultforBurrowTravel _hitresult))) (.TargetAttachComponent (Ability|GameplayCue|BreakGameplayCueParameters (BurrowTravelGameplay|MakeGameplayCueParamfromHitResultforBurrowTravel _hitresult))) (.bReplicateLocationWhenUsingMinimalRepProxy (Ability|GameplayCue|BreakGameplayCueParameters (BurrowTravelGameplay|MakeGameplayCueParamfromHitResultforBurrowTravel _hitresult))))))
            (:Targeted
              (Ability|ApplyGameplayEffectContainer _targetdata _applicationtag))))))
    (else
      (Utilities|Time|ClearandInvalidateTimerbyHandle (Variables|Transient|BurrowTravelGameplay|GetBurrowTravelPulseTimerHandle))
      (CallFunction|FullBurrowTravelGameplayComplete))))

(fn FindLandscapeAtLocation
  (params (Location :type Vector))
  (Utilities|FlowControl|Sequence
    (:then_0
      (bind _outhits (Collision|MultiLineTraceForObjects (Utilities|Operators|Add Location (Math|Vector|MakeVector 0.0 0.0 (ScalableFloat|GetCachedScalableFloatValue (Variables|Hotfix|GetHotfix_BurrowTravelLandscapeTraceZOffsetUp)))) (Utilities|Operators|Subtract Location (Math|Vector|MakeVector 0.0 0.0 (ScalableFloat|GetCachedScalableFloatValue (Variables|Hotfix|GetHotfix_BurrowTravelLandscapeTraceZOffsetDown)))) (Utilities|Array|MakeArray "ObjectTypeQuery1" "ObjectTypeQuery13")))
      (for _array_element _outhits
        (if (Math|Boolean|ORBoolean (.bSuccess (Utilities|Casting|CastToLandscape (.HitActor (Collision|BreakHitResult _array_element)))) (.bSuccess (Utilities|Casting|CastToLandscapeProxy (.HitActor (Collision|BreakHitResult _array_element)))) (.bSuccess (Utilities|Casting|CastToLandscapeSplineActor (.HitActor (Collision|BreakHitResult _array_element)))) (.bSuccess (Utilities|Casting|CastToFortStaticMeshActor (.HitActor (Collision|BreakHitResult _array_element)))) (.bSuccess (Utilities|Casting|CastToFortWaterBodyActor (.HitActor (Collision|BreakHitResult _array_element)))))
          (return true _array_element))))
    (:then_1
      (return false 0))))

(fn GetCustomAbilitySourceTransform
  (return (Math|Conversions|ToTransform(Vector) (Variables|Transient|BurrowTravelGameplay|GetLatestBurrowTravelPulseLocation))))

(fn CanActivateAbility
  (params (ActorInfo :type Gameplay%20Ability%20Actor%20Info%20Structure) (Handle :type Gameplay%20Ability%20Spec%20Handle%20Structure))
  (bind _currenttargetactor (Default|GetCurrentTargetActor(Message) (Ability|GetAvatarActorfromActorInfo)))
  (return 0 (Utilities|IsValid _currenttargetactor :!class /Script/Engine.KismetSystemLibrary)))

(fn MakeGameplayCueParamFromHitResultForBurrowTravel
  (params (HitResult :type Hit%20Result%20Structure%20%28by%20ref%29))
  (bind (_normalizedmagnitude _rawmagnitude _effectcontext _matchedtagname _originaltag _aggregatedsourcetags _aggregatedtargettags _location _normal _instigator _effectcauser _sourceobject _physicalmaterial _gameplayeffectlevel _abilitylevel _targetattachcomponent _breplicatelocationwhenusingminimalrepproxy) (Ability|GameplayCue|BreakGameplayCueParameters (GameplayCue|MakeGameplayCueParametersfromHitResult HitResult)))
  (bind _output_get (LocalVariables|SetEffectContext (GameplayEffects|MakeEffectContext (Ability|GetAbilitySystemComponentfromActorInfo))))
  (Ability|EffectContext|AddHitResult _output_get HitResult)
  (return (Ability|GameplayCue|MakeGameplayCueParameters _normalizedmagnitude _rawmagnitude _output_get _matchedtagname _originaltag _aggregatedsourcetags _aggregatedtargettags _location _normal _instigator _effectcauser _sourceobject _physicalmaterial _gameplayeffectlevel _abilitylevel _targetattachcomponent _breplicatelocationwhenusingminimalrepproxy)))

(fn SpawnDustTrail
  (params (OriginLocation :type Vector) (UnburrowLocation :type Vector))
  (bind _output_get (LocalVariables|SetDirection (Math|Vector|Normalize (- UnburrowLocation OriginLocation))))
  (bind _returnvalue (Game|SpawnActorfromClass "/ProtoSwarm/Blueprints/Medium/BP_TidalCrane_Medium_UnburrowTell.BP_TidalCrane_Medium_UnburrowTell_C" (Math|Transform|MakeTransform (+ (Variables|GetOriginLocation) (* _output_get 3e+02))) "AlwaysSpawn" "MultiplyWithRoot" (Variables|InternalData|GetBossActor) 0 0))
  (bind _output_get_2 (Variables|Default|SetBurrowTellActor _returnvalue))
  (Class|BPTidalCraneMediumUnburrowTell|SetUnborrowLocation _output_get_2 (+ (* (LocalVariables|GetDirection) 5e+03) (Variables|GetOriginLocation))))
