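; Gameplay ability that plays the Tidal Crane boss's spawn-in.
; It turns the boss toward any valid target, disables look-at and plays a montage alongside a dust cue, base setup and knockback, then unhides the boss.
; A BodySlam montage notify applies effect containers via target selection, with a custom source transform ahead of and above the boss; other notifies restore look-at.
; Ending the ability starts the boss's brain and re-enables look-at.
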(event Ability|EventActivateAbility
  (bind (_montageinfo_montagetoplay _montageinfo_animplayrate _montageinfo_animrootmotiontranslationscale _montageinfo_montagesectiontoplay _montageinfo_overridesection) (Utilities|Struct|BreakFortGameplayAbilityMontageInfo (Variables|GetMontageInfo_MontageToPlay)))
  (Ability|CommitAbility)
  (Utilities|FlowControl|Delay 0.2)
  (bind _output_get (Variables|InternalData|SetBossActor (Ability|GetAvatarActorfromActorInfo)))
  (bind _currenttargetactor (Default|GetCurrentTargetActor(Message) _output_get))
  (bind _output_get_2 (Variables|InternalData|SetCurrentTarget _currenttargetactor))
  (Utilities|IsValid _output_get_2
    (:Is%20Valid
      (CallFunction|RotateTowardsTarget))
    (:Is%20Not%20Valid
      _))
  (Default|SetAllowLookatTarget(Message) (Variables|InternalData|GetBossActor) false)
  (bind (_asynctaskproxy _notifytag) (Ability|Tasks|PlayMontageAndWaitNotify "None" _montageinfo_montagetoplay _montageinfo_animplayrate _montageinfo_overridesection true _montageinfo_animrootmotiontranslationscale)
    (:then
      (Ability|ExecuteGameplayCueOnOwner (Variables|Default|GetGC_SpawnInDust))
      (Default|HandleSetupBase(Message) (Variables|InternalData|GetBossActor))
      (CallFunction|TryKnockbackwithGameplayEffect)
      (Utilities|FlowControl|Delay 0.2)
      (Rendering|SetActorHiddenInGame (Variables|InternalData|GetBossActor) false))
    (:OnCompleted
      _)
    (:OnInterrupted
      _)
    (:OnCancelled
      _)
    (:OnNotifyBegin
      (CallFunction|NotifyBeginTag _notifytag)
      (if (== _notifytag "(TagName=\"Event.TidalCrane.Boss.BodySlam\")")
        (bind (_asynctaskproxy_2 _targetdata _applicationtag) (Ability|Tasks|DoTargetSelection "None" "(TagName=\"Targeting.ApplicationTag.1\")" true)
          (:Targeted
            (Ability|ApplyGameplayEffectContainer _targetdata _applicationtag)))
        (else
          (Default|SetAllowLookatTarget(Message) (Variables|InternalData|GetBossActor) true)))))
  (Ability|EndAbility))

(event Ability|EventOnEndAbility
  (params (bWasCancelled :type Boolean))
  (bind _bossactor (Variables|InternalData|GetBossActor))
  ; End Ability
  (Utilities|IsValid _bossactor
    (:Is%20Valid
      (Default|HandleStartBrain(Message) _bossactor)
      (Default|SetAllowLookatTarget(Message) _bossactor true))))

(event Custom|NotifyBeginTag
  (params (NotifyTag :type Gameplay%20Tag%20Structure)))

(event Custom|RotateTowardsTarget
  ; Rotate the Boss towards it's target
  (Default|HandleForceLookatLocationonSpawn(Message) (Variables|InternalData|GetBossActor)))

(fn GetCustomAbilitySourceTransform
  (return (Math|Transform|MakeTransform (+ (+ (Transformation|GetActorLocation (Ability|GetAvatarActorfromActorInfo)) (* (Transformation|GetActorForwardVector (Ability|GetAvatarActorfromActorInfo)) 7e+02)) "0.000000,0.000000,150.000000") (Transformation|GetActorRotation (Ability|GetAvatarActorfromActorInfo)))))
