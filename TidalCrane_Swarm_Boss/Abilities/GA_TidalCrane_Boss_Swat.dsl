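; Tidal Crane swarm boss swat ability that damages and knocks away nearby targets with a sideways swipe.
; Activation requires an in-range potential target other than the current one.
; An in-range target's side picks the montage section, gameplay cue level and lateral offset of the custom source transform.
; The swatting notify runs target selection, applies the effect container, then launches pawns or routes other actors to vehicle knockback.
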(event Ability|EventActivateAbility
  (bind (_montageinfo_montagetoplay _montageinfo_animplayrate _montageinfo_animrootmotiontranslationscale) (Utilities|Struct|BreakFortGameplayAbilityMontageInfo (Variables|GetMontageInfo_MontageToPlay)))
  (Ability|CommitAbility)
  (Variables|InternalData|SetBossActor (Ability|GetAvatarActorfromActorInfo))
  (bind _selectedactor (CallFunction|FindFirstSwattableTarget))
  (Variables|InternalData|SetSwipeLeft (CallFunction|GetSwipeDirection _selectedactor))
  (bind (_asynctaskproxy _notifytag) (Ability|Tasks|PlayMontageAndWaitNotify "None" _montageinfo_montagetoplay _montageinfo_animplayrate (select (Variables|InternalData|GetSwipeLeft) "None" "SwatRight") true _montageinfo_animrootmotiontranslationscale)
    (:OnCompleted
      _)
    (:OnInterrupted
      _)
    (:OnCancelled
      _)
    (:OnNotifyBegin
      (if (== _notifytag "(TagName=\"Event.TidalCrane.Boss.Swatting\")")
        (bind (_asynctaskproxy_2 _targetdata _applicationtag) (Ability|Tasks|DoTargetSelection "None" "(TagName=\"Targeting.ApplicationTag.0\")")
          (:then
            (Ability|ExecuteGameplayCueWithParamsOnOwner (Variables|Default|GetGC_SwatAway) (Ability|GameplayCue|MakeGameplayCueParameters :NormalizedMagnitude 0.0 :RawMagnitude 0.0 :OriginalTag (Variables|Default|GetGC_SwatAway) :Location "0,0,0" :Normal "0,0,0" :GameplayEffectLevel (select (Variables|InternalData|GetSwipeLeft) 2 1) :AbilityLevel (select (Variables|InternalData|GetSwipeLeft) 2 1) :bReplicateLocationWhenUsingMinimalRepProxy false)))
          (:Targeted
            (CallFunction|ApplyDamageandKnockback _targetdata _applicationtag))))))
  (Ability|EndAbility))

(event Ability|EventOnEndAbility
  (params (bWasCancelled :type Boolean))
  (Default|SetAllowLookatTarget(Message) (Variables|InternalData|GetBossActor) true))

(event Custom|ApplyDamageAndKnockback
  (params (TargetData :type Gameplay%20Ability%20Target%20Data%20Handle%20Structure) (ApplicationTag :type Gameplay%20Tag%20Structure))
  (Ability|ApplyGameplayEffectContainer TargetData ApplicationTag)
  (for _array_element (Ability|TargetData|GetAllActorsfromTargetData TargetData)
    (bind _asfort_pawn (Utilities|Casting|CastToFortPawn _array_element)
      (:then
        (CallFunction|KnockbackPawns _asfort_pawn))
      (:CastFailed
        (bind _asfort_athena_vehicle (Utilities|Casting|CastToFortAthenaVehicle _array_element))
        (CallFunction|KnockbackVehicles _asfort_athena_vehicle)))))

(fn FindFirstSwattableTarget
  (bind _bossactor (Variables|InternalData|GetBossActor))
  (Variables|InternalData|SetSwatTarget 0)
  (bind _potentialtargets (Default|GetPotentialTargets(Message) _bossactor))
  (bind _array_element (Utilities|Array|ForEachLoopwithBreak _potentialtargets)
    (:LoopBody
      (if (<= (Math|Vector|Distance(Vector) (Transformation|GetActorLocation _array_element) (Transformation|GetActorLocation _bossactor)) (ScalableFloat|GetValueatLevel (Variables|Default|GetMaximumDistanceForAttack)))
        (Variables|InternalData|SetSwatTarget _array_element)))
    (:Completed
      _))
  (return (Variables|InternalData|GetSwatTarget)))

(fn GetCustomAbilitySourceTransform
  (return (Math|Transform|MakeTransform (+ (+ (Transformation|GetActorLocation (Ability|GetAvatarActorfromActorInfo)) (* (Transformation|GetActorRightVector (Ability|GetAvatarActorfromActorInfo)) (select (Variables|InternalData|GetSwipeLeft) 4.5e+02 -4.5e+02))) "0.000000,0.000000,150.000000"))))

(fn GetSwipeDirection
  (params (ActorToTest :type Actor%20Object%20Reference))
  (return (not (< (Math|Vector|DotProduct (Transformation|GetActorRightVector (Ability|GetAvatarActorfromActorInfo)) (Math|Vector|Normalize (- (Transformation|GetActorLocation ActorToTest) (Transformation|GetActorLocation (Ability|GetAvatarActorfromActorInfo))))) 0))))

(fn CanActivateAbility
  (params (ActorInfo :type Gameplay%20Ability%20Actor%20Info%20Structure) (Handle :type Gameplay%20Ability%20Spec%20Handle%20Structure))
  (bind _returnvalue (Ability|GetAvatarActorfromActorInfo))
  (bind _currenttargetactor (Default|GetCurrentTargetActor(Message) _returnvalue))
  (bind _potentialtargets (Default|GetPotentialTargets(Message) _returnvalue))
  (LocalVariables|SetCurrentTarget _currenttargetactor)
  (bind _array_element (Utilities|Array|ForEachLoopwithBreak _potentialtargets)
    (:LoopBody
      (if (and (<= (Math|Vector|Distance(Vector) (Transformation|GetActorLocation _array_element) (Transformation|GetActorLocation _returnvalue)) (ScalableFloat|GetValueatLevel (Variables|Default|GetMaximumDistanceForAttack))) (!= _array_element (LocalVariables|GetCurrentTarget)))
        (return 0 true)))
    (:Completed
      (return 0 false))))

(fn KnockbackPawns
  (params (Pawn :type Fort%20Pawn%20Object%20Reference))
  (bind (_returnvalue_x _returnvalue_y) (Math|Vector|BreakVector (Math|Vector|Normalize (- (Transformation|GetActorLocation Pawn) (Transformation|GetActorLocation (Variables|InternalData|GetBossActor))))))
  (Character|LaunchCharacter Pawn (Math|Vector|MakeVector (* _returnvalue_x (ScalableFloat|GetValueatLevel (Variables|Default|GetKnockbackPawnForwardVelocity))) (* _returnvalue_y (ScalableFloat|GetValueatLevel (Variables|Default|GetKnockbackPawnForwardVelocity))) (ScalableFloat|GetValueatLevel (Variables|Default|GetKnockbackPawnHeightVelocity))) true true))

(fn KnockbackVehicles
  (params (FortAthenaVehicle :type Fort%20Athena%20Vehicle%20Object%20Reference))
  (bind (_returnvalue_x _returnvalue_y) (Math|Vector|BreakVector (Math|Vector|Normalize (- (Transformation|GetActorLocation FortAthenaVehicle) (Transformation|GetActorLocation (Variables|InternalData|GetBossActor))))))
  (Audio|BroadcastAbilityImpactatLocation 0 0 (Math|Vector|MakeVector (* _returnvalue_x (ScalableFloat|GetValueatLevel (Variables|Default|GetKnockbackVehicleForwardVelocity))) (* _returnvalue_y (ScalableFloat|GetValueatLevel (Variables|Default|GetKnockbackVehicleForwardVelocity))) (ScalableFloat|GetValueatLevel (Variables|Default|GetKnockbackVehicleHeightVelocity))) 777.0 "None" false))
