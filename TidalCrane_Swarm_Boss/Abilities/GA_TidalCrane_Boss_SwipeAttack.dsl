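; Gameplay ability for the TidalCrane swarm boss swipe attack, activatable only when its current target is within a maximum 2D distance.
; It disables look-at until ability end, adds a telegraph cue, waits, then plays a montage whose notifies trigger target selection.
; A custom source transform sits ahead of and above the boss.
; Hits apply effect containers, launch pawns away, and pass vehicle knockback vectors to an impact broadcast call.
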(event Ability|EventActivateAbility
  (bind (_montageinfo_montagetoplay _montageinfo_animplayrate _montageinfo_animrootmotiontranslationscale _montageinfo_montagesectiontoplay _montageinfo_overridesection) (Utilities|Struct|BreakFortGameplayAbilityMontageInfo (Variables|GetMontageInfo_MontageToPlay)))
  (Ability|CommitAbility)
  (Default|SetAllowLookatTarget(Message) (Ability|GetAvatarActorfromActorInfo) false)
  (Ability|AddGameplayCueToOwner (Variables|Default|GetGC_BodySlamTelegraph))
  (Utilities|FlowControl|Delay (ScalableFloat|GetValueatLevel (Variables|Default|GetAnticipationDelay)))
  (bind (_asynctaskproxy _notifytag) (Ability|Tasks|PlayMontageAndWaitNotify "None" _montageinfo_montagetoplay _montageinfo_animplayrate _montageinfo_overridesection true _montageinfo_animrootmotiontranslationscale)
    (:OnCompleted
      _)
    (:OnInterrupted
      _)
    (:OnCancelled
      _)
    (:OnNotifyBegin
      (bind (_asynctaskproxy_2 _targetdata _applicationtag) (Ability|Tasks|DoTargetSelection "None" _notifytag)
        (:Targeted
          (CallFunction|ApplyDamageandKnockback _targetdata _applicationtag)))))
  (Ability|EndAbility))

(event Ability|EventOnEndAbility
  (params (bWasCancelled :type Boolean))
  (Default|SetAllowLookatTarget(Message) (Ability|GetAvatarActorfromActorInfo) true))

(event Custom|ApplyDamageAndKnockback
  (params (TargetData :type Gameplay%20Ability%20Target%20Data%20Handle%20Structure) (ApplicationTag :type Gameplay%20Tag%20Structure))
  (Ability|ApplyGameplayEffectContainer TargetData ApplicationTag)
  (for _array_element (Ability|TargetData|GetAllActorsfromTargetData TargetData)
    (bind _asfort_pawn (Utilities|Casting|CastToFortPawn _array_element)
      (:then
        (CallFunction|KnockbackPawns _asfort_pawn))
      (:CastFailed
        (bind _asfort_athena_vehicle (Utilities|Casting|CastToFortAthenaVehicle _array_element))
        (CallFunction|KnockbackVehicles _asfort_athena_vehicle)))))

(fn GetCustomAbilitySourceTransform
  (return (Math|Transform|MakeTransform (+ (+ (Transformation|GetActorLocation (Ability|GetAvatarActorfromActorInfo)) (* (Transformation|GetActorForwardVector (Ability|GetAvatarActorfromActorInfo)) 3e+03)) "0.000000,0.000000,500.000000") (Transformation|GetActorRotation (Ability|GetAvatarActorfromActorInfo)))))

(fn CanActivateAbility
  (params (ActorInfo :type Gameplay%20Ability%20Actor%20Info%20Structure) (Handle :type Gameplay%20Ability%20Spec%20Handle%20Structure))
  (bind _returnvalue (Ability|GetAvatarActorfromActorInfo))
  (bind _currenttargetactor (Default|GetCurrentTargetActor(Message) _returnvalue))
  (Utilities|IsValid _currenttargetactor
    (:Is%20Valid
      (return 0 (<= (Math|Vector|Distance2D(Vector) (Transformation|GetActorLocation _returnvalue) (Transformation|GetActorLocation _currenttargetactor)) (ScalableFloat|GetValueatLevel (Variables|Default|GetMaximumDistanceForAttack)))))
    (:Is%20Not%20Valid
      (return 0 false))))

(fn KnockbackPawns
  (params (Pawn :type Fort%20Pawn%20Object%20Reference))
  (bind (_returnvalue_x _returnvalue_y) (Math|Vector|BreakVector (Math|Vector|Normalize (- (Transformation|GetActorLocation Pawn) (Transformation|GetActorLocation (Ability|GetAvatarActorfromActorInfo))))))
  (Character|LaunchCharacter Pawn (Math|Vector|MakeVector (* _returnvalue_x (ScalableFloat|GetValueatLevel (Variables|Default|GetKnockbackPawnForwardVelocity))) (* _returnvalue_y (ScalableFloat|GetValueatLevel (Variables|Default|GetKnockbackPawnForwardVelocity))) (ScalableFloat|GetValueatLevel (Variables|Default|GetKnockbackPawnHeightVelocity))) true true))

(fn KnockbackVehicles
  (params (FortAthenaVehicle :type Fort%20Athena%20Vehicle%20Object%20Reference))
  (bind (_returnvalue_x _returnvalue_y) (Math|Vector|BreakVector (Math|Vector|Normalize (- (Transformation|GetActorLocation FortAthenaVehicle) (Transformation|GetActorLocation (Ability|GetAvatarActorfromActorInfo))))))
  (Audio|BroadcastAbilityImpactatLocation 0 0 (Math|Vector|MakeVector (* _returnvalue_x (ScalableFloat|GetValueatLevel (Variables|Default|GetKnockbackVehicleForwardVelocity))) (* _returnvalue_y (ScalableFloat|GetValueatLevel (Variables|Default|GetKnockbackVehicleForwardVelocity))) (ScalableFloat|GetValueatLevel (Variables|Default|GetKnockbackVehicleHeightVelocity))) 777.0 "None" false))
