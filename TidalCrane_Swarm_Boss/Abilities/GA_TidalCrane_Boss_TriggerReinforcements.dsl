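; Gameplay ability for the TidalCrane swarm boss that summons small or large reinforcements through a montage.
; A scalable-float flag picks the variant, and the StartReinforcements notify messages the boss to spawn that type.
; Look-at targeting turns off on activation and returns on any other notify or ability end.
; Activation requires boss permission, no current reinforcements and available Living World Manager event budget.
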(event Ability|EventActivateAbility
  (bind (_montageinfo_montagetoplay _montageinfo_animplayrate _montageinfo_animrootmotiontranslationscale _montageinfo_montagesectiontoplay _montageinfo_overridesection) (Utilities|Struct|BreakFortGameplayAbilityMontageInfo (Variables|GetMontageInfo_MontageToPlay)))
  (bind _bossactor (Variables|InternalData|GetBossActor))
  (bind _bossactor_2 (Variables|InternalData|GetBossActor))
  (Ability|CommitAbility)
  (bind _output_get (Variables|InternalData|SetBossActor (Ability|GetAvatarActorfromActorInfo)))
  (Default|SetAllowLookatTarget(Message) _output_get false)
  (if (ScalableFloat|GetValueasBool (Variables|Default|GetSpawnSmallReinforcements))
    (bind (_asynctaskproxy _notifytag) (Ability|Tasks|PlayMontageAndWaitNotify "None" _montageinfo_montagetoplay _montageinfo_animplayrate _montageinfo_overridesection true _montageinfo_animrootmotiontranslationscale)
      (:OnCompleted
        _)
      (:OnInterrupted
        _)
      (:OnCancelled
        _)
      (:OnNotifyBegin
        (if (== _notifytag (Variables|Default|GetTag_StartReinforcements))
          (Default|HandleTriggerReinforcements(Message) _bossactor_2 true)
          (else
            (Default|SetAllowLookatTarget(Message) _bossactor_2 true)))))
    (Ability|EndAbility)
    (else
      (bind (_asynctaskproxy_2 _notifytag_2) (Ability|Tasks|PlayMontageAndWaitNotify "None" (Variables|Default|GetMontageLargeReinforcements))
        (:OnCompleted
          _)
        (:OnInterrupted
          _)
        (:OnCancelled
          _)
        (:OnNotifyBegin
          (if (== _notifytag_2 (Variables|Default|GetTag_StartReinforcements))
            (Default|HandleTriggerReinforcements(Message) _bossactor false)
            (else
              (Default|SetAllowLookatTarget(Message) _bossactor true)))))
      (Ability|EndAbility))))

(event Ability|EventOnEndAbility
  (params (bWasCancelled :type Boolean))
  (bind _bossactor (Variables|InternalData|GetBossActor))
  (Utilities|IsValid _bossactor
    (:Is%20Valid
      (Default|SetAllowLookatTarget(Message) _bossactor true))))

(fn CanActivateAbility
  (params (ActorInfo :type Gameplay%20Ability%20Actor%20Info%20Structure) (Handle :type Gameplay%20Ability%20Spec%20Handle%20Structure))
  (bind _returnvalue (Ability|GetAvatarActorfromActorInfo))
  (bind _currentreinforcementscount (Default|GetCurrentReinforcementsCount(Message) _returnvalue))
  (bind _cantriggerreinforcements (Default|GetCanTriggerReinforcements(Message) _returnvalue))
  (return 0 (Math|Boolean|ANDBoolean _cantriggerreinforcements (<= _currentreinforcementscount 0) (> (Class|FortAthenaLivingWorldManager|QueryEventBudget (Class|FortAthenaLivingWorldManager|GetLivingWorldManager) (Variables|Default|GetQueenReinforcementsLagerEntry) _returnvalue) 0))))
