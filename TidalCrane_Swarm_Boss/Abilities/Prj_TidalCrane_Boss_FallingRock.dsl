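; Falling rock projectile for the TidalCrane swarm boss that can drop ammo when it explodes.
; BeginPlay plays spawn VFX and sound where cosmetic events can run, and a hit destroys the projectile's Niagara component.
; On explosion, the authority removes the target-location and volley target-location gameplay cues from the owner.
; It also rolls against a scalable chance to spawn one of five random ammo loot tiers at the rock's location.
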(event EventOnExploded
  (params (HitActors :type Array%20of%20Actor%20Object%20References) (HitResults :type Array%20of%20Hit%20Result%20Structures))
  ; Spawn Loot on Explosion (Server)
  (|CallParentFunction HitActors HitResults)
  (Networking|SwitchHasAuthority
    (:Authority
      (bind _asbuilding_gameplay_actor (Utilities|Casting|CastToBuildingGameplayActor (Actor|GetOwner))
        (:then
          (GameplayCue|RemoveGameplayCue _asbuilding_gameplay_actor (Variables|Default|GetGC_ProjectileTargetLocation))
          (GameplayCue|RemoveGameplayCue _asbuilding_gameplay_actor (Variables|Default|GetGC_VolleyProjectileTargetLocation)))
        (:CastFailed
          _))
      (if (<= (Math|Random|RandomFloatinRange 0.0 1.0) (ScalableFloat|GetValueatLevel (Variables|Default|GetChanceToSpawnAmmo)))
        (Inventory|SpawnPickupInWorldWithLootTier :LootTierName (Utilities|Select :Index (Math|Random|RandomIntegerinRange 0 4) :Option%200 "LTG_Vend_Item_Pistol_Ammo" :Option%201 "LTG_Vend_Item_Shotgun_Ammo" :Option%202 "LTG_Vend_Item_SMG_Ammo" :Option%203 "LTG_Vend_Item_Assault_Ammo" :Option%204 "LTG_Vend_Item_Sniper_Ammo") :Position (Transformation|GetActorLocation self) :OverrideMaxStackCount 0 :bToss false :bTossWithVelocity false :SourceType "AI" :Source "ItemSpawner" :LootWorldLevel 0)))))

(event EventBeginPlay
  (|CallParentFunction)
  (Networking|CanExecuteCosmeticEvents
    (:True
      (Niagara|SpawnSystematLocation (Variables|Default|GetSpawnVFX%20) (Transformation|GetActorLocation))
      (Audio|PlaySoundatLocation "/TidalCrane_Swarm_Boss/Assets/Sound/SC_QueenBug_Rock_Spawn.SC_QueenBug_Rock_Spawn" (Transformation|GetActorLocation)))))

(event Collision|EventHit
  (params (MyComp :type Primitive%20Component%20Object%20Reference) (Other :type Actor%20Object%20Reference) (OtherComp :type Primitive%20Component%20Object%20Reference) (bSelfMoved :type Boolean) (HitLocation :type Vector :default "0, 0, 0") (HitNormal :type Vector :default "0, 0, 0") (NormalImpulse :type Vector :default "0, 0, 0") (Hit :type Hit%20Result%20Structure%20%28by%20ref%29))
  (|CallParentFunction MyComp Other OtherComp bSelfMoved HitLocation HitNormal NormalImpulse Hit)
  (Components|DestroyComponent (Variables|Default|GetNS_TidalCrane_Prj)))

(fn ConstructionScript
  (|CallParentFunction))
