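; Ranged projectile for the TidalCrane swarm boss that can drop ammo where it explodes.
; On explosion it runs the parent handling, then removes the projectile and volley target-location gameplay cues from its owner, cast to a Building Gameplay Actor.
; On authority only, a random roll against a scalable-float chance spawns a pickup from a randomly selected pistol, shotgun, SMG, assault or sniper ammo loot tier.
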(event EventOnExploded
  (params (HitActors :type Array%20of%20Actor%20Object%20References) (HitResults :type Array%20of%20Hit%20Result%20Structures))
  ; Spawn Loot on Explosion (Server)
  (|CallParentFunction HitActors HitResults)
  (bind _asbuilding_gameplay_actor (Utilities|Casting|CastToBuildingGameplayActor (Actor|GetOwner))
    (:then
      (GameplayCue|RemoveGameplayCue _asbuilding_gameplay_actor (Variables|Default|GetGC_ProjectileTargetLocation))
      (GameplayCue|RemoveGameplayCue _asbuilding_gameplay_actor (Variables|Default|GetGC_VolleyProjectileTargetLocation)))
    (:CastFailed
      _))
  (Networking|SwitchHasAuthority
    (:Authority
      (if (<= (Math|Random|RandomFloatinRange 0.0 1.0) (ScalableFloat|GetValueatLevel (Variables|Default|GetChanceToSpawnAmmo)))
        (Inventory|SpawnPickupInWorldWithLootTier :LootTierName (Utilities|Select :Index (Math|Random|RandomIntegerinRange 0 4) :Option%200 "LTG_Vend_Item_Pistol_Ammo" :Option%201 "LTG_Vend_Item_Shotgun_Ammo" :Option%202 "LTG_Vend_Item_SMG_Ammo" :Option%203 "LTG_Vend_Item_Assault_Ammo" :Option%204 "LTG_Vend_Item_Sniper_Ammo") :Position (Transformation|GetActorLocation self) :OverrideMaxStackCount 0 :bToss false :bTossWithVelocity false :SourceType "AI" :Source "ItemSpawner" :LootWorldLevel 0)))))

(fn ConstructionScript
  (|CallParentFunction))
